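# Unlit translucent capture material for ImpostorBaker that draws scene buffers as post-process-style output, with depth testing disabled.
# MPC_ImposterCapture weights multiply each GBuffer channel, world normal and a depth term from position in the capture basis, then sum them.
# Custom nodes read ClearCoat from GBuffer CustomData and mask subsurface color for that shading model.
# EnableCapture blends this over scene color, with a small green screen-corner dot and a radius-based mask around Origin.
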Material /ImpostorBaker/Materials/PostProcess/M_Capture_PP_Via_Translucency
graph {
 properties {
  bDisableDepthTest=true
  bUseTranslucencyVertexFog=false
  blendMode=BLEND_Translucent
  pixelDepthOffsetMode=PDOM_Legacy
  shadingModel=MSM_Unlit
  usedShadingModels=Unlit
 }
 nodes [
  CollectionParameter() {collection=/ImpostorBaker/Materials/MPC/MPC_ImposterCapture, parameterName=Emissive}
  n1 = SceneColor!(UVs=ScreenPosition().ViewportUV)
  n2 = Multiply(
   A=ComponentMask(None=SceneTexture() {sceneTextureId=PPI_StoredBaseColor}.Color) {b=true, g=true, r=true}
   B=CollectionParameter() {collection=&MPC_ImposterCapture, parameterName=BaseColor}
  )
  n3 = Multiply(
   A=ComponentMask(None=SceneTexture() {sceneTextureId=PPI_Metallic}.Color) {r=true}
   B=CollectionParameter() {collection=&MPC_ImposterCapture, parameterName=Metallic}
  )
  n4 = Multiply(
   A=ComponentMask(None=SceneTexture() {sceneTextureId=PPI_StoredSpecular}.Color) {r=true}
   B=CollectionParameter() {collection=&MPC_ImposterCapture, parameterName=Specular}
  )
  n5 = Multiply(
   A=ComponentMask(None=SceneTexture() {sceneTextureId=PPI_Roughness}.Color) {r=true}
   B=CollectionParameter() {collection=&MPC_ImposterCapture, parameterName=Roughness}
  )
  n6 = Multiply(
   A=ComponentMask(None=SceneTexture() {sceneTextureId=PPI_Anisotropy}.Color) {r=true}
   B=CollectionParameter() {collection=&MPC_ImposterCapture, parameterName=Anisotropy}
  )
  n7 = Multiply(B=CollectionParameter() {collection=&MPC_ImposterCapture, parameterName=Emissive})
  n8 = Multiply(
   A=ComponentMask(None=SceneTexture() {sceneTextureId=PPI_Opacity}.Color) {r=true}
   B=CollectionParameter() {collection=&MPC_ImposterCapture, parameterName=Opacity}
  )
  n9 = Multiply(
   A=ComponentMask(None=SceneTexture() {sceneTextureId=PPI_WorldTangent}.Color) {b=true, g=true, r=true}
   B=CollectionParameter() {collection=&MPC_ImposterCapture, parameterName=Tangent}
  )
  n10 = Add(A=Add(A=Add(A=@n2, B=@n3), B=Add(A=@n4, B=@n5)), B=Add(A=Add(A=@n6, B=@n7), B=Add(A=@n8, B=@n9)))
  MaterialExpressionComponentMask_2 = ComponentMask(None=SceneTexture() {sceneTextureId=PPI_ShadingModelID}.Color) {r=true}
  n11 = Custom!(ShadingModelID=@MaterialExpressionComponentMask_2) {
   code="float SubsurfaceMask = ShadingModelID == 4 ? 0: 1;\r\n\r\nreturn SubsurfaceMask;"
   description=Subsurface ID Mask
   inputs [{
     inputName=ShadingModelID
     input {expression=&self:MaterialExpressionComponentMask_2}
    }]
  }
  n12 = Multiply(A=ComponentMask(None=SceneTexture() {sceneTextureId=PPI_SubsurfaceColor}.Color) {b=true, g=true, r=true}, B=@n11) {desc=Ignore Subsurface for ClearCoat}
  n13 = CollectionParameter() {collection=&MPC_ImposterCapture, parameterName=SubSurface}
  n14 = CollectionParameter() {collection=&MPC_ImposterCapture, parameterName=ClearCoat}
  n15 = Custom(ShadingModelID=@MaterialExpressionComponentMask_2) {
   code="float2 UV = MaterialFloat2(ScreenAlignedPosition(Parameters.ScreenPosition).xy);\r\nFGBufferData GBData = GetGBufferData(UV, false);\r\n\r\nfloat ClearCoatMask = ShadingModelID == 4 ? 1: 0;\r\n\r\nClearCoat = GBData.CustomData.r * ClearCoatMask;\r\nClearCoatRoughness = GBData.CustomData.g * ClearCoatMask;\r\n\r\nreturn GBData.CustomData;"
   description=Gbuffer CustomData
   additionalOutputs [{outputName=ClearCoat, outputType=CMOT_Float1}, {outputName=ClearCoatRoughness, outputType=CMOT_Float1}]
   inputs [{
     inputName=ShadingModelID
     input {expression=&self:MaterialExpressionComponentMask_2}
    }]
  }
  n16 = CollectionParameter() {collection=&MPC_ImposterCapture, parameterName=ClearCoatRoughness}
  n17 = ComponentMask(None=SceneTexture() {sceneTextureId=PPI_AmbientOcclusion}.Color) {r=true}
  n18 = CollectionParameter() {collection=&MPC_ImposterCapture, parameterName=AmbientOcclusion}
  n19 = Add(
   A=Add(A=Multiply(A=@n12, B=@n13), B=Multiply(A=@n15.ClearCoat, B=@n14))
   B=Add(A=Multiply(A=@n15.ClearCoatRoughness, B=@n16), B=Multiply(A=@n17, B=@n18))
  )
  n20 = CollectionParameter() {collection=&MPC_ImposterCapture, parameterName=Normal}
  n21 = ConstantBiasScale(None=ComponentMask(None=SceneTexture() {sceneTextureId=PPI_WorldNormal}.Color) {b=true, g=true, r=true})
  n22 = MaterialFunctionCall() {materialFunction=&WorldPositionBehindTranslucency}
  n23 = ComponentMask(None=CollectionParameter() {collection=&MPC_ImposterCapture, parameterName=Origin}) {b=true, g=true, r=true}
  n24 = CollectionParameter() {collection=&MPC_ImposterCapture, parameterName=X}
  n25 = CollectionParameter() {collection=&MPC_ImposterCapture, parameterName=Y}
  n26 = CollectionParameter() {collection=&MPC_ImposterCapture, parameterName=Z}
  n27 = Divide(
   A=MaterialFunctionCall!(
    VectorToTransform=Subtract(A=@n22.Result, B=@n23)
    BasisX=ComponentMask(None=@n24) {b=true, g=true, r=true}
    BasisY=ComponentMask(None=@n25) {b=true, g=true, r=true}
    BasisZ=ComponentMask(None=@n26) {b=true, g=true, r=true}
   ) {materialFunction=/Engine/Functions/Engine_MaterialFunctions02/Math/InverseTransformMatrix}.Transformed Vector 3
   B=CollectionParameter() {collection=&MPC_ImposterCapture, parameterName=R}
  )
  n28 = Saturate(None=ComponentMask(None=ConstantBiasScale(None=@n27)) {b=true})
  n29 = Add(
   A=Add(A=Add(A=@n10, B=@n19), B=Multiply(A=@n20, B=@n21))
   B=Multiply(A=OneMinus(None=@n28), B=CollectionParameter() {collection=&MPC_ImposterCapture, parameterName=Depth})
  )
  n30 = CollectionParameter() {collection=&MPC_ImposterCapture, parameterName=EnableCapture}
  n31 = ScreenPosition()
  n32 = MaterialFunctionCall(Float2=ViewProperty() {property=MEVP_ViewSize}.Property) {materialFunction=&BreakOutFloat2Components}
  n33 = AppendVector(A=Divide(A=@n32.G, B=@n32.G), B=Constant() {r=1})
  n34 = VectorParameter() {defaultValue=<rgba 0.05 0.95 0 0>}
  LinearInterpolate(
   A=LinearInterpolate(A=@n1.RGB, B=@n29, Alpha=@n30)
   B=Constant3Vector() {constant=<rgba 0 0.5 0 0>}
   Alpha=SphereMask(A=Divide(A=@n31.ViewportUV, B=@n33), B=ComponentMask(None=@n34.RGB) {g=true, r=true}) {attenuationRadius=0.025}
  )
  n35 = MaterialFunctionCall(Float3=@n1.RGB) {materialFunction=&BreakOutFloat3Components}
  MaterialFunctionCall!(Linear=@n35.R) {materialFunction=/Engine/Functions/Engine_MaterialFunctions01/ImageAdjustment/LinearTosRGB}
  MaterialFunctionCall(Linear=@n35.G) {materialFunction=&LinearTosRGB}
  MaterialFunctionCall(Linear=@n35.B) {materialFunction=&LinearTosRGB}
  MaterialFunctionCall(
   X=MaterialFunctionCall!(sRGB=@n35.R) {materialFunction=/Engine/Functions/Engine_MaterialFunctions01/ImageAdjustment/sRGBToLinear}.Linear
   Y=MaterialFunctionCall(sRGB=@n35.G) {materialFunction=&sRGBToLinear}.Linear
   Z=MaterialFunctionCall(sRGB=@n35.B) {materialFunction=&sRGBToLinear}.Linear
  ) {materialFunction=&MakeFloat3}
  n36 = MaterialFunctionCall() {materialFunction=&WorldPositionBehindTranslucency}
  n37 = ComponentMask(None=CollectionParameter() {collection=&MPC_ImposterCapture, parameterName=Origin}) {b=true, g=true, r=true}
  n38 = Subtract(
   A=Multiply(A=CollectionParameter() {collection=&MPC_ImposterCapture, parameterName=R}) {constB=64}
   B=Length(None=Subtract(A=@n36.Result, B=@n37)) {is=class MaterialExpressionLength}
  )
  Multiply(A=@n30, B=Saturate(None=Ceil(None=@n38)))
  Power(Base=@n29) {constExponent=2.2}
  ScalarParameter() {defaultValue=0.9, parameterName=gsfghdfsdf}
 ]
}
