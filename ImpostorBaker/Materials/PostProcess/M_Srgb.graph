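# M_Srgb is an unlit post-process material in the ImpostorBaker folder that reads the scene image.
# Its graph samples PostProcessInput0 through a SceneTexture node and masks the color to RGB, dropping alpha.
# Despite the sRGB name, the graph contains no gamma or color-space math.
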Material /ImpostorBaker/Materials/PostProcess/M_Srgb
graph {
 properties {
  bDisableDepthTest=true
  blendMode=BLEND_Translucent
  materialDomain=MD_PostProcess
  pixelDepthOffsetMode=PDOM_Legacy
  shadingModel=MSM_Unlit
  usedShadingModels=Unlit
 }
 nodes [ComponentMask(None=SceneTexture() {sceneTextureId=PPI_PostProcessInput0}.Color) {b=true, g=true, r=true}]
}
